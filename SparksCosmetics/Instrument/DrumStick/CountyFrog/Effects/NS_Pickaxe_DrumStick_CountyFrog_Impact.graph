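# Niagara impact effect for the CountyFrog drumstick pickaxe: a short blue watery burst with bubbles, droplets and surface debris.
# Flash merges several spawn groups into one emitter with sprites, mesh and light, alongside an expanding, distance-culled Ring.
# Bubbles wobble and rise before a late-life splash, droplets shrink as they fall, and debris responds to User.PhysicalSurface.
# Front-end flags and time of day adjust brightness, alpha and burst origin.
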NiagaraSystem /SparksCosmetics/Instrument/DrumStick/CountyFrog/Effects/NS_Pickaxe_DrumStick_CountyFrog_Impact
dependencies {
 usedDataInterfaces [&NiagaraDataInterfaceCurve, &NiagaraDataInterfaceColorCurve]
 usedDynamicInputs [&SimulationPosition, &UniformRangedVector2D]
 usedModules [
  &self:SystemSpawnScript.SystemScriptSource.SystemScriptGraph.NiagaraNodeAssignment_0.SetVariables_#1
  &SystemState
 ]
}
emitters {
 Bubble_Sprites001 {
  emitterUpdateScript [
   &EmitterState()
   &SpawnBurst_Instantaneous(
    Spawn Count=&RandomRangeInt(Minimum=4, Maximum=6)
    Use Spawn Probability=false
    UseLoopCountLimit=false
   )
  ]
  particleSpawnScript [
   /SetVariables(Particles.ScaleMod:NiagaraFloat=&FloatFromCurve(
     FloatCurve=DataInterface() {propertyValues {Curve {keys [
         {arriveTangent=-0.2041, interpMode=RCIM_Cubic, leaveTangent=-0.2041, tangentMode=RCTM_User, time=0.1, value=1}
         {arriveTangent=-0.6837, interpMode=RCIM_Cubic, leaveTangent=-0.6837, tangentMode=RCTM_User, time=0.9, value=0.5}
        ]}}}
     CurveIndex=&RandomRangeFloat(Minimum=0, Maximum=1)
    )) {isSetParameters=true}
   &InitializeParticle(
    Color=&RandomRangeLinearColor(Minimum=<rgba 1 1 1 1>, Maximum=<rgba 1 1 1 0.5>)
    Lifetime=&FloatFromCurve(
     FloatCurve=DataInterface() {propertyValues {Curve {keys [<at 0 0.6>,<at 1 0.3>]}}}
     CurveIndex=&ReturnNormalizedExecIndex()
    )
    Sprite Size Mode=::NewEnumerator3
    Uniform Sprite Size=&Multiply_Float(A=&Multiply_Float(A=25, B=$Particles.ScaleMod), B=$System.GlobalScale)
    UsePositionOffset=false
    Write Material Random=true
   )
   &ShapeLocation(
    MaskSpawnGroup=false
    Sphere Distribution=::NewEnumerator3
    Sphere Radius=&Multiply_Float(A=6, B=$System.GlobalScale)
    UseOwnerScale=true
   )
   &AddVelocityFromPoint(
    Module.Use Falloff Distance=false
    Origin Offset=&FN_FrontEndVector3Modifier(Gameplay=<-2 0 0>, Front End=<0 0 0>)
    Velocity Strength=&Multiply_Float(A=&RandomRangeFloat(Minimum=150, Maximum=500), B=$System.GlobalScale)
   )
   /SetVariables(Particles.float:NiagaraFloat=&RandomRangeFloat(Minimum=0, Maximum=360)) {isSetParameters=true}
   &CameraOffset(Camera Offset Amount=20)
  ]
  particleUpdateScript [
   &ParticleState()
   &ScaleSpriteSize(
    Scale Factor=&Multiply_Vector2D_ByFloat(
     Float=&FloatFromCurve(
      FloatCurve=DataInterface() {propertyValues {Curve {
         defaultValue=1
         keys [
          {arriveTangent=0.0844, interpMode=RCIM_Cubic, leaveTangent=0.0844, leaveTangentWeight=0.3333, tangentMode=RCTM_User, time=0, value=0.53}
          {
           arriveTangent=0.3354
           arriveTangentWeight=0.1911
           interpMode=RCIM_Cubic
           leaveTangent=0.3354
           leaveTangentWeight=0.1422
           tangentMode=RCTM_User
           time=0.5734
           value=0.5966
          }
          {
           arriveTangent=0.7849
           arriveTangentWeight=0.0271
           interpMode=RCIM_Cubic
           leaveTangent=0.7849
           leaveTangentWeight=0.1151
           tangentMode=RCTM_User
           time=0.6546
           value=0.6043
          }
          {
           arriveTangent=0.6802
           arriveTangentWeight=0.0306
           interpMode=RCIM_Cubic
           leaveTangent=0.6802
           leaveTangentWeight=0.0846
           tangentMode=RCTM_User
           time=0.7463
           value=0.6851
          }
          {
           arriveTangent=0.4774
           arriveTangentWeight=0.0394
           interpMode=RCIM_Cubic
           leaveTangent=0.4774
           leaveTangentWeight=0.0452
           tangentMode=RCTM_User
           time=0.8644
           value=0.6515
          }
          {
           arriveTangent=12.3514
           arriveTangentWeight=0.0119
           interpMode=RCIM_Cubic
           leaveTangent=12.3514
           leaveTangentWeight=0.0333
           tangentMode=RCTM_User
           time=0.9
           value=1
          }
          {interpMode=RCIM_Cubic, tangentMode=RCTM_User, time=1, value=1}
         ]
        }}}
      Scale Curve=$System.GlobalScale
     )
     Vector2D=&MakeVector2D(
      X=&FloatFromCurve(FloatCurve=DataInterface() {propertyValues {Curve {keys [
           {interpMode=RCIM_Cubic, leaveTangentWeight=0.2737, tangentMode=RCTM_User, time=0, value=1.35}
           {arriveTangent=-4.0805, interpMode=RCIM_Cubic, leaveTangent=-4.0805, tangentMode=RCTM_User, time=0.1813, value=1}
           {arriveTangent=-0.1492, interpMode=RCIM_Cubic, leaveTangent=-0.1492, time=0.2013, value=0.8992}
           <at 0.2613 0.9881 RCIM_Cubic>
          ]}}})
      Y=&FloatFromCurve(FloatCurve=DataInterface() {propertyValues {Curve {keys [
           {arriveTangent=-0.2248, interpMode=RCIM_Cubic, leaveTangent=-0.2248, tangentMode=RCTM_User, time=0, value=0.75}
           {arriveTangent=3.3994, interpMode=RCIM_Cubic, leaveTangent=3.3994, tangentMode=RCTM_User, time=0.1813, value=1}
           <at 0.2013 1.1617 RCIM_Cubic>
           <at 0.2613 1 RCIM_Cubic>
          ]}}})
     )
    )
    Scale Sprite Size Mode=::NewEnumerator2
   )
   &Drag(Drag=&RandomRangeFloat(Minimum=9, Maximum=13), Use Linear Drag=true, Use Rotational Drag=false)
   &GravityForce(Gravity=&MakeVector(
     X=0
     Y=0
     Z=&FloatFromCurve(
      FloatCurve=DataInterface() {propertyValues {Curve {keys [<at 0.8 95 RCIM_Constant>,<at 0.9 -200 RCIM_Constant>]}}}
      Scale Curve=&RandomRangeFloat(Minimum=0.9, Maximum=1.2)
     )
    ))
   &CurlNoiseForce(
    Bias Noise=false
    Noise Strength=&FloatFromCurve(
     FloatCurve=DataInterface() {propertyValues {Curve {
        defaultValue=1
        keys [
         {arriveTangent=1.2375, interpMode=RCIM_Cubic, leaveTangent=1.2375, tangentMode=RCTM_User, time=0.1, value=0}
         {arriveTangentWeight=0.2399, interpMode=RCIM_Cubic, leaveTangentWeight=0.2938, tangentMode=RCTM_User, time=0.5, value=1}
         {arriveTangent=-6.6694, interpMode=RCIM_Cubic, leaveTangent=-6.6694, tangentMode=RCTM_User, time=0.9, value=0}
        ]
       }}}
     Scale Curve=75
    )
    Pan Noise=false
   )
   &SolveForcesAndVelocity(Clamp Velocity=false, Limit Acceleration=false)
   &ScaleColor(
    Scale Alpha=&Multiply_Float(
     A=&Multiply_Float(
      A=&FloatFromCurve(FloatCurve=DataInterface() {propertyValues {Curve {keys [
           {arriveTangent=0.8727, interpMode=RCIM_Cubic, leaveTangent=0.8727, tangentMode=RCTM_User, time=0, value=0}
           {arriveTangent=1.859, interpMode=RCIM_Cubic, leaveTangent=1.859, tangentMode=RCTM_User, time=0.2882, value=0.2618}
           {
            arriveTangent=13.7023
            arriveTangentWeight=0.1345
            interpMode=RCIM_Cubic
            leaveTangentWeight=0.1345
            tangentMode=RCTM_Break
            time=0.4
            value=1
           }
           {arriveTangentWeight=0.1274, interpMode=RCIM_Cubic, leaveTangentWeight=0.2613, tangentMode=RCTM_User, time=0.8, value=1}
           {arriveTangent=-18.5888, tangentMode=RCTM_User, time=0.9, value=0}
          ]}}})
      B=&FloatFromCurve(
       FloatCurve=DataInterface() {propertyValues {Curve {keys [<at 0 0.75>,<at 0.3925 1>,<at 0.5964 1>,<at 1 0.75>]}}}
       CurveIndex=Linked() {linkedVariable {name=NPC.FortniteNPC.TimeOfDay_0-1, type=NiagaraFloat}}
       Scale Curve=&MakeCustomFloatFromBool(Bool=$User.bIsFrontEndPreview, True Float=0.7, False Float=1)
      )
     )
     B=1
    )
    Scale RGB=&Multiply_VectorByFloat(Vector=<1 1 1>, Float=1)
    ScaleA=true
    ScaleRGB=true
   )
   /SetVariables(
    Particles.Splash Color:LinearColor=&MakeLinearColorFromVectorAndFloat(
     "Vector (RGB)"=<1 1 1>
     "Float (Alpha)"=&FloatFromCurve(
      FloatCurve=DataInterface() {propertyValues {Curve {keys [
          <at 0.88 0 RCIM_Cubic>
          {arriveTangent=-38.7858, interpMode=RCIM_Cubic, leaveTangent=-38.7858, tangentMode=RCTM_User, time=0.9, value=2}
          <at 1 0 RCIM_Cubic>
         ]}}}
      Scale Curve=&MakeCustomFloatFromBool(Bool=$User.bIsFrontEndPreview, True Float=0.7, False Float=1)
     )
    )
    Particles.Splash Dyn 0:Vector4f=&MakeVector4(
     X=&FloatFromCurve(FloatCurve=DataInterface() {propertyValues {Curve {
         defaultValue=1
         keys [
          {interpMode=RCIM_Cubic, tangentMode=RCTM_User, time=0.9, value=0}
          {arriveTangent=14.9612, interpMode=RCIM_Cubic, leaveTangent=14.9612, tangentMode=RCTM_User, time=1, value=0.5786}
         ]
        }}})
     Y=0
     Z=0
     W=0
    )
    Particles.Splash Size:Vector2f=&Multiply_Vector2D(
     A=$Particles.Initial.SpriteSize
     B=&Vector2DFromFloat(Value=&FloatFromCurve(
       FloatCurve=DataInterface() {propertyValues {Curve {keys [
           {arriveTangent=19.4142, interpMode=RCIM_Constant, time=0.85, value=0}
           {arriveTangentWeight=1.0543, interpMode=RCIM_Cubic, leaveTangentWeight=14.76, tangentMode=RCTM_User, time=0.9, value=0.5}
           {arriveTangent=-1.8786, interpMode=RCIM_Cubic, leaveTangent=-1.8786, tangentMode=RCTM_User, time=1, value=1}
          ]}}}
       Scale Curve=1.75
      ))
    )
   ) {isSetParameters=true}
   &ScaleMeshSize(Scale Factor=&Multiply_VectorByFloat(
     Vector=<1 1 1>
     Float=&FloatFromCurve(FloatCurve=DataInterface() {propertyValues {Curve {keys [
          {arriveTangent=0.0844, interpMode=RCIM_Cubic, leaveTangent=0.0844, tangentMode=RCTM_User, time=0, value=0.5}
          {arriveTangent=0.3354, interpMode=RCIM_Cubic, leaveTangent=0.3354, tangentMode=RCTM_User, time=0.5734, value=0.5966}
          {arriveTangent=0.7849, interpMode=RCIM_Cubic, leaveTangent=0.7849, tangentMode=RCTM_User, time=0.6546, value=0.6043}
          {arriveTangent=0.6802, interpMode=RCIM_Cubic, leaveTangent=0.6802, tangentMode=RCTM_User, time=0.7463, value=0.6851}
          {arriveTangent=0.4774, interpMode=RCIM_Cubic, leaveTangent=0.4774, tangentMode=RCTM_User, time=0.8644, value=0.6515}
          {arriveTangent=12.3514, interpMode=RCIM_Cubic, leaveTangent=12.3514, tangentMode=RCTM_User, time=0.9, value=1}
         ]}}})
    ))
   &DynamicMaterialParameters(
    Index 0 Param 1=&FloatFromCurve(FloatCurve=DataInterface() {propertyValues {Curve {
        defaultValue=1
        keys [
         {arriveTangent=14.7642, interpMode=RCIM_Cubic, leaveTangent=14.7642, tangentMode=RCTM_User, time=0, value=0}
         {arriveTangent=-2.5249, interpMode=RCIM_Cubic, leaveTangent=-2.5249, tangentMode=RCTM_User, time=0.1, value=0.7}
         {arriveTangent=-0.0405, interpMode=RCIM_Cubic, leaveTangent=-0.0405, tangentMode=RCTM_User, time=0.5, value=0}
        ]
       }}})
    Index 0 Param 2=&FloatFromCurve(FloatCurve=DataInterface() {propertyValues {Curve {
        defaultValue=1
        keys [
         {interpMode=RCIM_Cubic, leaveTangentWeight=0.9127, tangentMode=RCTM_User, time=0.2, value=0}
         {arriveTangentWeight=0.2716, interpMode=RCIM_Cubic, leaveTangentWeight=0.4526, tangentMode=RCTM_User, time=0.5, value=0.2}
         {arriveTangent=-0.0405, interpMode=RCIM_Cubic, leaveTangent=-0.0405, tangentMode=RCTM_User, time=1, value=0}
        ]
       }}})
    Index 0 Param 3=&FloatFromCurve(FloatCurve=DataInterface() {propertyValues {Curve {keys [<at 0.25 1>,<at 0.3 0>]}}})
    Module.Index0Float0Write=true
    Module.Index0Float1Write=true
    Module.Index0Float2Write=true
    Module.Index0Float3Write=true
   )
  ]
  rendererClasses [&NiagaraSpriteRendererProperties]
  renderers [
   {rendererClass=&NiagaraSpriteRendererProperties}
   {rendererClass=&NiagaraSpriteRendererProperties, rendererIndex=1}
  ]
 }
 Drops001 {
  emitterUpdateScript [
   &EmitterState()
   &SpawnBurst_Instantaneous(
    Spawn Count=&RandomRangeInt(Minimum=17, Maximum=20)
    Use Spawn Probability=false
    UseLoopCountLimit=false
   )
  ]
  particleSpawnScript [
   &InitializeParticle(
    Color=<rgba 0.2167 0.835 2 1>
    Lifetime=&UniformRangedFloat(Minimum=0.2, Maximum=0.4)
    Mesh Scale=&Multiply_VectorByFloat(Vector=<1 1 1>, Float=$System.GlobalScale)
    Mesh Scale Mode=::NewEnumerator4
    UsePositionOffset=false
    Write Material Random=true
   ) {module=InitializeParticle001}
   &ShapeLocation(
    MaskSpawnGroup=false
    Override Seed=false
    Sphere Radius=&Multiply_Float(A=6, B=$System.GlobalScale)
    UseOwnerScale=true
   )
   &AddVelocityFromPoint(
    Module.Use Falloff Distance=false
    Origin Offset=&FN_FrontEndVector3Modifier(Gameplay=<-2 0 0>, Front End=<0 0 0>)
    Velocity Strength=&Multiply_Float(A=&RandomRangeFloat(Minimum=150, Maximum=500), B=$System.GlobalScale)
   )
  ]
  particleUpdateScript [
   &ParticleState()
   &ScaleMeshSize(Scale Factor=&VectorFromFloat(Value=&FloatFromCurve(
      FloatCurve=DataInterface() {propertyValues {Curve {keys [
          {interpMode=RCIM_Cubic, tangentMode=RCTM_User, time=0, value=1}
          {arriveTangent=-2.5355, interpMode=RCIM_Cubic, leaveTangent=-2.5355, tangentMode=RCTM_User, time=1, value=0}
         ]}}}
      Scale Curve=&RandomRangeFloat(Minimum=0.9, Maximum=2)
     )))
   &Drag(Drag=4, Use Linear Drag=true, Use Rotational Drag=true)
   &GravityForce(Gravity=&Multiply_VectorByFloat(
     Vector=<0 0 -900>
     Float=&FloatFromCurve(FloatCurve=DataInterface() {propertyValues {Curve {defaultValue=1, keys [<at 0 0>,<at 0.2 0>,<at 0.25 1>,<at 1 0>]}}})
    ))
   &SolveForcesAndVelocity(Clamp Velocity=false, Limit Acceleration=false)
   &ScaleColor(
    Scale Alpha=&FloatFromCurve(
     FloatCurve=DataInterface() {propertyValues {Curve {keys [<at 0 1>,<at 0.7 1>,<at 1 0>]}}}
     Scale Curve=&RandomRangeFloat(Minimum=1, Maximum=1)
    )
    Scale RGB=&FN_FrontEndVector3Modifier(Gameplay=<1 1 1>, Front End=<1.5 1.5 1.5>)
    ScaleA=true
    ScaleRGB=true
   )
  ]
  rendererClasses [&NiagaraMeshRendererProperties]
  renderers [{rendererClass=&NiagaraMeshRendererProperties, rendererIndex=0}]
 }
 FN_PhysMatDebris {
  emitterUpdateScript [
   &EmitterState()
   &SpawnBurst_Instantaneous(
    Spawn Count=&RandomRangeInt(Minimum=5, Maximum=10)
    Use Spawn Probability=false
    UseLoopCountLimit=false
   )
  ]
  particleSpawnScript [
   /SetVariables(Particles.BiasedRandom=&FloatFromCurve(
     FloatCurve=DataInterface() {propertyValues {Curve {keys [
         <at 0.2 1 RCIM_Cubic>
         {arriveTangent=-3.3394, interpMode=RCIM_Cubic, leaveTangent=-3.3394, tangentMode=RCTM_User, time=0.5, value=0.5}
         <at 0.8 0 RCIM_Cubic>
        ]}}}
     CurveIndex=&RandomRangeFloat(Minimum=0, Maximum=1)
    )) {isSetParameters=true}
   &InitializeParticle(
    Color Mode=::NewEnumerator0
    Lifetime=&RandomRangeFloat(Minimum=0.5, Maximum=1.25)
    Mass=&Lerp_Float(A=0.6, B=1.2, Alpha=$Particles.BiasedRandom)
    Mass Mode=::NewEnumerator1
    Mesh Scale=&Multiply_VectorByFloat(
     Vector=&RandomRangeVector(Minimum=<0.9 0.4 0.5>, Maximum=<1 1 1>)
     Float=&Lerp_Float(A=0.7, B=1, Alpha=$Particles.BiasedRandom)
    )
    Mesh Scale Mode=::NewEnumerator4
    UsePositionOffset=false
    Write Material Random=true
   )
   &FN_PhysicalSurface(
    Dirt Mesh Count=2
    Glass Mesh Count=3
    Ice Mesh Count=2
    Metal Gravity=&RandomRangeFloat(Minimum=-350, Maximum=-1800)
    Metal Mesh Count=2
    Metal Velocity=&RandomRangeFloat(Minimum=80, Maximum=375)
    Metal Velocity Cone Angle=130
    Physical Surface=$User.PhysicalSurface
    Stone Gravity=&RandomRangeFloat(Minimum=-700, Maximum=-1250)
    Stone Mesh Count=2
    Stone Velocity=&RandomRangeFloat(Minimum=80, Maximum=550)
    Stone Velocity Cone Angle=120
    Wood Gravity=&RandomRangeFloat(Minimum=-450, Maximum=-900)
    Wood Mesh Count=2
    Wood Scale=1.5
    Wood Velocity=&RandomRangeFloat(Minimum=80, Maximum=600)
    Wood Velocity Cone Angle=130
   )
   &AddVelocityInCone(
    Cone Angle=$Output.FN_PhysicalSurface.SurfaceResponseConeAngle
    Module.Use Velocity Falloff On Cone Axis=false
    Velocity Strength=$Output.FN_PhysicalSurface.SurfaceResponseVelocity
   )
   &InitialMeshOrientation(Use Rotation Vector=false)
   &AddRotationalVelocity(Rotation Rate=&RandomVector(Vector Scale=&RandomRangeFloat(Minimum=1, Maximum=2)))
   &ApplyInitialForces()
   /SetVariables(Particles.GravityOverrideValue=$Output.FN_PhysicalSurface.GravityOverride) {isSetParameters=true}
  ]
  particleUpdateScript [
   &ParticleState()
   &GravityForce(Gravity=$Particles.GravityOverrideValue)
   &Collision(
    Advanced Aging Rate=1
    Age Colliding Particles=false
    CPU Collision Trace Channel=ECC_WorldStatic
    Collision Enabled=&Boolean_AndOperation(
     Bool1=&SetBoolByFloatComparison(A=$Particles.Initial.Lifetime, B=1)
     Bool2=&SetBoolByFloatComparison(A=$Particles.Age, B=0.12)
    )
    Control Roll On Collision=true
    Friction=1
    Friction During a Bounce=0.25
    Max CPU Trace Length=100
    Module.Control Rest State Via Surface Normal=false
    Module.EnableMaxCollisionCount=false
    Module.Randomize Collision Normal Vector=true
    Randomize Collision Normal=0.25
    ScaleRateOfRotation=false
   ) {enabled=false}
   &SolveForcesAndVelocity(Clamp Velocity=false, Limit Acceleration=false)
   &ScaleMeshSize(Scale Factor=&VectorFromFloat(Value=&FloatFromCurve(FloatCurve=DataInterface() {propertyValues {Curve {keys [<at 0.6 1 RCIM_Cubic>, {arriveTangent=-6.308, interpMode=RCIM_Cubic, leaveTangent=-6.308, tangentMode=RCTM_User, time=1, value=0}]}}})))
   &DynamicMaterialParameters(Module.Param0WriteEnabled=true, Module.Param1WriteEnabled=true, Module.Param2WriteEnabled=false, Module.Param3WriteEnabled=false)
  ]
  rendererClasses [&NiagaraSpriteRendererProperties, &NiagaraMeshRendererProperties]
  renderers [
   {rendererClass=&NiagaraSpriteRendererProperties}
   {rendererClass=&NiagaraMeshRendererProperties, rendererIndex=1}
   {rendererClass=&NiagaraMeshRendererProperties, rendererIndex=2}
   {rendererClass=&NiagaraMeshRendererProperties, rendererIndex=3}
   {rendererClass=&NiagaraMeshRendererProperties, rendererIndex=4}
   {rendererClass=&NiagaraMeshRendererProperties, rendererIndex=5}
   {rendererClass=&NiagaraMeshRendererProperties, rendererIndex=6}
   {rendererClass=&NiagaraMeshRendererProperties, rendererIndex=7}
   {rendererClass=&NiagaraMeshRendererProperties, rendererIndex=8}
  ]
 }
 Flash {
  emitterUpdateScript [
   &EmitterState()
   &SpawnBurst_Instantaneous(Spawn Count=1, Use Spawn Probability=false)
   &SpawnBurst_Instantaneous(Spawn Count=&RandomRangeInt(Minimum=7, Maximum=9), Spawn Group=1, Use Spawn Probability=false) {module=SpawnBurst_Instantaneous001}
   &SpawnBurst_Instantaneous(Spawn Count=5, Spawn Group=2, Use Spawn Probability=false, UseLoopCountLimit=false) {module=SpawnBurst_Instantaneous002}
   &SpawnBurst_Instantaneous(Spawn Group=3, Use Spawn Probability=false, UseLoopCountLimit=false) {module=SpawnBurst_Instantaneous003}
   &SpawnBurst_Instantaneous(Spawn Group=4, Use Spawn Probability=false, UseLoopCountLimit=false) {module=SpawnBurst_Instantaneous004}
   &SpawnBurst_Instantaneous(Spawn Group=5, Use Spawn Probability=false, UseLoopCountLimit=false) {module=SpawnBurst_Instantaneous005}
  ]
  particleSpawnScript [
   /SetVariables(Particles.VisibilityTag=&SpawnGroup()) {isSetParameters=true}
   &InitializeParticle(
    Color=&MergeHelper_ColorSwitch(
     0=<rgba 0.3206 1.171 2 1>
     1=<rgba 0.3594 0.6161 1 1>
     2=<rgba 0.1467 0.3113 0.8047 1>
     3=<rgba 0 0.4915 1 1>
     4=<rgba 1 1 1 1>
     5=<rgba 0.0005 0.0233 0.1146 0.75>
     Index=&SpawnGroup()
     Num=::NewEnumerator4
    )
    Lifetime=&MergeHelper_FloatSwitch(
     0=0.19
     1=&RandomRangeFloat(Minimum=0.1, Maximum=0.13)
     2=&RandomRangeFloat(Minimum=0.09, Maximum=0.12)
     3=0.2
     4=0.15
     5=0.2
     Index=&SpawnGroup()
     Num=::NewEnumerator4
    )
    Mesh Scale=&Multiply_VectorByFloat(Vector=&RandomRangeVector(Minimum=<1 2 2>, Maximum=<0.75 1.5 1.5>), Float=$System.GlobalScale)
    Mesh Scale Mode=::NewEnumerator4
    Sprite Size=&Multiply_Vector2D_ByFloat(
     Float=$System.GlobalScale
     Vector2D=&MergeHelper_Vector2Switch(
      0=<1 1>
      1=&RandomRangeVector2D(Minimum=<9 45>, Maximum=<15 70>)
      2=&RandomRangeVector2D(Minimum=<30 45>, Maximum=<45 60>)
      3=&Vector2DFromFloat(Value=200)
      4=<1 1>
      5=&Vector2DFromFloat(Value=125)
      Index=$Emitter.SpawnGroup
      Num=::NewEnumerator4
     )
    )
    Sprite Size Mode=::NewEnumerator4
    UsePositionOffset=false
    Write Material Random=true
   )
   &DynamicMaterialParameters(
    Index 0 Param 2=45
    Index 0 Param 3=1
    Index 0 Param 4=1
    Module.Param0WriteEnabled=false
    Module.Param1WriteEnabled=true
    Module.Param2WriteEnabled=true
    Module.Param3WriteEnabled=true
   ) {module=DynamicMaterialParameters001}
   &InitialMeshOrientation(
    Rotation=&RandomRangeVector(Minimum=<0 0 0>, Maximum=<1 1 1>)
    Use Orientation Vector=true
    Use Rotation Vector=true
   )
   &CameraOffset(Camera Offset Amount=&MergeHelper_FloatSwitch(0=20, 1=10, 2=10, 3=20, 4=0, 5=25, Index=&SpawnGroup(), Num=::NewEnumerator4))
   &AddVelocityFromPoint(
    Module.Use Falloff Distance=false
    Velocity Strength=&MergeHelper_FloatSwitch(0=0, 1=100, 2=125, 3=0, 4=0, Index=&SpawnGroup(), Num=::NewEnumerator3)
   )
   &ShapeLocation(
    MaskSpawnGroup=false
    Override Seed=false
    Sphere Radius=&MergeHelper_FloatSwitch(
     0=0
     1=&Multiply_Float(A=6, B=$System.GlobalScale)
     2=&Multiply_Float(A=6, B=$System.GlobalScale)
     3=0
     4=0
     5=0
     Index=&SpawnGroup()
     Num=::NewEnumerator4
    )
    UseOwnerScale=true
   )
  ]
  particleUpdateScript [
   &ParticleState()
   &DynamicMaterialParameters(
    Index 0 Param 1=&FloatFromCurve(FloatCurve=DataInterface() {propertyValues {Curve {keys [<at 0.2 0 RCIM_Cubic>, {arriveTangent=3.6451, interpMode=RCIM_Cubic, leaveTangent=3.6451, tangentMode=RCTM_User, time=1, value=1}]}}})
    Module.Index0Float0Write=true
    Module.Index0Float1Write=true
    Module.Index0Float2Write=true
    Module.Index0Float3Write=true
   )
   &ScaleMeshSize(Scale Factor=&VectorFromFloat(Value=&FloatFromCurve(
      FloatCurve=DataInterface() {propertyValues {Curve {
         defaultValue=0.5
         keys [
          {arriveTangent=2.4982, interpMode=RCIM_Cubic, leaveTangent=2.4982, tangentMode=RCTM_User, time=0, value=0}
          {arriveTangent=1.2726, interpMode=RCIM_Cubic, leaveTangent=1.2726, tangentMode=RCTM_User, time=1, value=1}
         ]
        }}}
      Scale Curve=0.7
     )))
   &ScaleColor(
    Scale Alpha=&MergeHelper_FloatSwitch(
     0=&FloatFromCurve(FloatCurve=DataInterface() {propertyValues {Curve {keys [<at 0 1>,<at 1 0>]}}})
     1=&FloatFromCurve(FloatCurve=DataInterface() {propertyValues {Curve {
         defaultValue=1
         keys [{interpMode=RCIM_Cubic, tangentMode=RCTM_User, time=0, value=1}, {interpMode=RCIM_Cubic, tangentMode=RCTM_User, time=1, value=0}]
        }}})
     2=&FloatFromCurve(FloatCurve=DataInterface() {propertyValues {Curve {
         defaultValue=1
         keys [
          {arriveTangent=-2.0249, interpMode=RCIM_Cubic, leaveTangent=-2.0249, tangentMode=RCTM_User, time=0, value=1}
          {interpMode=RCIM_Cubic, tangentMode=RCTM_User, time=1, value=0}
         ]
        }}})
     3=&FloatFromCurve(
      FloatCurve=DataInterface() {propertyValues {Curve {
         defaultValue=1
         keys [
          {interpMode=RCIM_Cubic, tangentMode=RCTM_User, time=0, value=1}
          {arriveTangent=-1.2989, interpMode=RCIM_Cubic, leaveTangent=-1.2989, tangentMode=RCTM_User, time=1, value=0}
         ]
        }}}
      Scale Curve=&FN_FrontEndFloatModifier(Gameplay=1, Front End=0.5)
     )
     4=1
     5=&FloatFromCurve(
      FloatCurve=DataInterface() {propertyValues {Curve {
         defaultValue=1
         keys [
          {interpMode=RCIM_Cubic, tangentMode=RCTM_User, time=0, value=1}
          {arriveTangent=-2.0745, interpMode=RCIM_Cubic, leaveTangent=-2.0745, tangentMode=RCTM_User, time=1, value=0}
         ]
        }}}
      Scale Curve=&FloatFromCurve(
       FloatCurve=DataInterface() {propertyValues {Curve {
          defaultValue=1
          keys [
           {interpMode=RCIM_Cubic, tangentMode=RCTM_User, time=0, value=0.45}
           {interpMode=RCIM_Cubic, tangentMode=RCTM_User, time=0.5, value=1}
           {tangentMode=RCTM_User, time=1, value=0.45}
          ]
         }}}
       CurveIndex=Linked() {linkedVariable {name=NPC.FortniteNPC.TimeOfDay_0-1, type=NiagaraFloat}}
       Scale Curve=1.2
      )
     )
     Index=&SpawnGroup()
     Num=::NewEnumerator4
    )
    Scale RGB=&MergeHelper_VectorSwitch(
     0=&Multiply_VectorByFloat(
      Vector=&MakeVectorFromLinearColorRGB(Color=&ColorFromCurve(ColorCurve=DataInterface() {propertyValues {
          AlphaCurve {keys []}
          BlueCurve {keys [<at 0 5 RCIM_Cubic>,<at 0.15 1>,<at 0.833 1 RCIM_Cubic>]}
          GreenCurve {keys [<at 0 4 RCIM_Cubic>,<at 0.15 0.6953>,<at 0.833 0.6016 RCIM_Cubic>]}
          RedCurve {keys [<at 0 2 RCIM_Cubic>,<at 0.15 0.3906>,<at 0.833 0.2031 RCIM_Cubic>]}
         }}))
      Float=&FloatFromCurve(FloatCurve=DataInterface() {propertyValues {Curve {keys [<at 0 4>,<at 0.2 1.5>,<at 1 1>]}}})
     )
     1=&MakeVectorFromLinearColorRGB(Color=&ColorFromCurve(ColorCurve=DataInterface() {propertyValues {
         AlphaCurve {keys []}
         BlueCurve {keys [<at 0 5 RCIM_Cubic>,<at 0.4312 1>,<at 1 1 RCIM_Cubic>]}
         GreenCurve {keys [<at 0 3.0166 RCIM_Cubic>,<at 0.4312 0.6953>,<at 1 0.6016 RCIM_Cubic>]}
         RedCurve {keys [<at 0 1.6033 RCIM_Cubic>,<at 0.4312 0.3906>,<at 1 0.2031 RCIM_Cubic>]}
        }}))
     2=&VectorFromFloat(Value=0.4)
     3=&VectorFromFloat(Value=&FloatFromCurve(FloatCurve=DataInterface() {propertyValues {Curve {keys [<at 0 5>,<at 1 1>]}}}))
     4=<1 1 1>
     5=<1 1 1>
     Index=&SpawnGroup()
     Num=::NewEnumerator4
    )
    ScaleA=true
    ScaleRGB=true
   )
   &ScaleSpriteSize(
    Scale Factor=&MergeHelper_Vector2Switch(
     0=<0 0>
     1=&MakeVector2D(
      X=&FloatFromCurve(FloatCurve=DataInterface() {propertyValues {Curve {
          defaultValue=1
          keys [
           {interpMode=RCIM_Cubic, tangentMode=RCTM_User, time=0, value=0}
           {arriveTangentWeight=0.2236, interpMode=RCIM_Cubic, leaveTangentWeight=0.1491, tangentMode=RCTM_User, time=0.7, value=1}
           {arriveTangent=-2.2754, interpMode=RCIM_Cubic, leaveTangent=-2.2754, tangentMode=RCTM_User, time=1, value=0.5}
          ]
         }}})
      Y=&FloatFromCurve(FloatCurve=DataInterface() {propertyValues {Curve {
          defaultValue=1
          keys [
           {arriveTangent=3.143, interpMode=RCIM_Cubic, leaveTangent=3.143, tangentMode=RCTM_User, time=0, value=0}
           {arriveTangent=0.8537, interpMode=RCIM_Cubic, leaveTangent=0.8537, tangentMode=RCTM_User, time=1, value=1}
          ]
         }}})
     )
     2=&Vector2DFromFloat(Value=&FloatFromCurve(FloatCurve=DataInterface() {propertyValues {Curve {
          defaultValue=1
          keys [
           {arriveTangent=0.9682, interpMode=RCIM_Cubic, leaveTangent=0.9682, tangentMode=RCTM_User, time=0, value=0}
           {interpMode=RCIM_Cubic, tangentMode=RCTM_User, time=1, value=1}
          ]
         }}}))
     3=&Multiply_Vector2D_ByFloat(
      Float=&FloatFromCurve(FloatCurve=DataInterface() {propertyValues {Curve {
          defaultValue=0
          keys [
           {interpMode=RCIM_Cubic, tangentMode=RCTM_User, time=0, value=1}
           {arriveTangent=-1.4808, interpMode=RCIM_Cubic, leaveTangent=-1.4808, tangentMode=RCTM_User, time=1, value=0}
          ]
         }}})
      Vector2D=<1 1>
     )
     4=<0 0>
     5=&Vector2DFromFloat(Value=&FloatFromCurve(FloatCurve=DataInterface() {propertyValues {Curve {
          defaultValue=0
          keys [
           {arriveTangent=2.4894, interpMode=RCIM_Cubic, leaveTangent=2.4894, tangentMode=RCTM_User, time=0, value=0}
           {interpMode=RCIM_Cubic, tangentMode=RCTM_User, time=1, value=1}
          ]
         }}}))
     Index=&SpawnGroup()
     Num=::NewEnumerator4
    )
    Scale Sprite Size Mode=::NewEnumerator2
   )
   &SolveForcesAndVelocity(Clamp Velocity=false, Limit Acceleration=false)
   /SetVariables(Particles.Flare Dyn 1:Vector4f=&MakeVector4(
     X=&FloatFromCurve(FloatCurve=DataInterface() {propertyValues {Curve {keys [<at 0 1 RCIM_Cubic>,<at 0.8 0 RCIM_Cubic>]}}})
     Y=2
     Z=1
     W=1
    )) {isSetParameters=true}
   &FN_Light_Attributes(
    Color=<rgba 0 0.7912 1 1>
    Create Color Attribute=true
    Create Exponent Attribute=true
    Create Radius Attribute=true
    Exponent=3
    Radius=95
    Scale Color=&FloatFromCurve(
     FloatCurve=DataInterface() {propertyValues {Curve {
        defaultValue=1
        keys [{interpMode=RCIM_Cubic, tangentMode=RCTM_User, time=0, value=1}, {interpMode=RCIM_Cubic, tangentMode=RCTM_User, time=1, value=0}]
       }}}
     Scale Curve=12
    )
   )
  ]
  rendererClasses [
   &NiagaraMeshRendererProperties
   &NiagaraSpriteRendererProperties
   &NiagaraLightRendererProperties
  ]
  renderers [
   {rendererClass=&NiagaraMeshRendererProperties, rendererIndex=0}
   {rendererClass=&NiagaraSpriteRendererProperties, rendererIndex=1}
   {rendererClass=&NiagaraSpriteRendererProperties, rendererIndex=2}
   {rendererClass=&NiagaraSpriteRendererProperties, rendererIndex=3}
   {rendererClass=&NiagaraLightRendererProperties, rendererIndex=4}
   {rendererClass=&NiagaraSpriteRendererProperties, rendererIndex=5}
  ]
 }
 Ring {
  emitterUpdateScript [
   &EmitterState(
    Enable Distance Culling=true
    Max Distance Response=::KillImmediately
    MaxDistance=3500
    Module.bUseMaxDistance=true
    Module.bUseMinDistance=false
    Scalability Mode=::NewEnumerator1
   )
   &SpawnBurst_Instantaneous(Spawn Count=1, Use Spawn Probability=false)
  ]
  particleSpawnScript [
   &InitializeParticle(
    Color=&MultiplyLinearColor(
     Linear Color=<rgba 0 1.3597 4.5 1>
     Scale Factor=&MakeVector4_FromLinearColor(LinearColor=$System.TOD Color)
    )
    Lifetime=0.2
    Sprite Rotation Mode=::NewEnumerator2
    Sprite Size Mode=::NewEnumerator3
    Uniform Sprite Size=&Multiply_Float(A=115, B=$System.GlobalScale)
    UsePositionOffset=false
    Write Material Random=true
   )
   &CameraOffset(Camera Offset Amount=25)
  ]
  particleUpdateScript [
   &ParticleState()
   &ScaleColor(
    Scale Alpha=&FloatFromCurve(
     FloatCurve=DataInterface() {propertyValues {Curve {keys [<at 0.5 1 RCIM_Cubic>, {arriveTangentWeight=0.5056, interpMode=RCIM_Cubic, tangentMode=RCTM_User, time=1, value=0}]}}}
     Scale Curve=&FN_FrontEndFloatModifier(Gameplay=0.7, Front End=0.2)
    )
    ScaleA=true
    ScaleRGB=true
   )
   &ScaleSpriteSize(Uniform Curve Sprite Scale=DataInterface() {propertyValues {Curve {
       defaultValue=0
       keys [
        {arriveTangent=2.2531, interpMode=RCIM_Cubic, leaveTangent=2.2531, tangentMode=RCTM_User, time=0, value=0.2}
        {arriveTangent=0.3236, interpMode=RCIM_Cubic, leaveTangent=0.3236, tangentMode=RCTM_User, time=1, value=1}
       ]
      }}})
   &DynamicMaterialParameters(
    Index 0 Param 1=&FloatFromCurve(FloatCurve=DataInterface() {propertyValues {Curve {keys [
         {arriveTangent=-0.5613, interpMode=RCIM_Cubic, leaveTangent=-0.5613, tangentMode=RCTM_User, time=0, value=1}
         {arriveTangent=-1.1815, interpMode=RCIM_Cubic, leaveTangent=-1.1815, tangentMode=RCTM_User, time=1, value=0.3}
        ]}}})
    Index 0 Param 2=1
    Index 0 Param 3=1
    Module.Index0Float0Write=true
    Module.Index0Float1Write=true
    Module.Index0Float2Write=true
    Module.Index0Float3Write=true
   )
  ]
  rendererClasses [&NiagaraSpriteRendererProperties]
  renderers [{rendererClass=&NiagaraSpriteRendererProperties}]
 }
}
summary {
 systemName=NS_Pickaxe_DrumStick_CountyFrog_Impact
 emitters [
  {
   emitterName=FN_PhysMatDebris
   rendererClasses [&NiagaraSpriteRendererProperties, &NiagaraMeshRendererProperties]
  }
  {
   emitterName=Flash
   rendererClasses [
    &NiagaraMeshRendererProperties
    &NiagaraSpriteRendererProperties
    &NiagaraLightRendererProperties
   ]
  }
  {emitterName=Ring, rendererClasses [&NiagaraSpriteRendererProperties]}
  {emitterName=Drops001, rendererClasses [&NiagaraMeshRendererProperties]}
  {emitterName=Bubble_Sprites001, rendererClasses [&NiagaraSpriteRendererProperties]}
 ]
 userVariables [
  User.bIsFrontEnd {type=NiagaraBool}
  User.bIsFrontEndPreview {type=NiagaraBool}
  User.PhysicalSurface {type=NiagaraInt32}
 ]
}
