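# Niagara swing trail for the CountyFrog drumstick pickaxe, combining two short-lived anim-trail ribbons with a burst of bubble sprites.
# Ribbons form a dark, narrower line and a wider trail tinted by time-of-day color.
# Bubbles spawn from socket motion, drift with buoyancy and curl noise, then drop with a late-life splash flash.
# Distance from the emitter fades and kills bubbles, while time-of-day and front-end modifiers adjust opacity.
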NiagaraSystem /SparksCosmetics/Instrument/DrumStick/CountyFrog/Effects/NS_Pickaxe_DrumStick_CountyFrog_Trail
dependencies {
 usedDataInterfaces [
  &NiagaraDataInterfaceCurve
  &NiagaraDataInterfaceCamera
  &NiagaraDataInterfaceColorCurve
  &NiagaraDataInterfaceSocketReader
 ]
 usedDynamicInputs [&ColorFromCurve]
 usedModules [
  &self:SystemSpawnScript.SystemScriptSource.SystemScriptGraph.NiagaraNodeAssignment_0.SetVariables_#1
  &self:SystemSpawnScript.SystemScriptSource.SystemScriptGraph.NiagaraNodeAssignment_1.SetVariables_#2
  &SystemState
 ]
}
emitters {
 Bubble_Sprites001 {
  emitterUpdateScript [
   &EmitterState(Life Cycle Mode=::NewEnumerator1, Loop Behavior=::NewEnumerator1, Loop Duration=1, UseLoopDelay=false)
   &SpawnRate(
    Spawn Probability=&FloatFromCurve(
     FloatCurve=DataInterface() {propertyValues {Curve {
        defaultValue=1
        keys [
         {arriveTangent=7.8977, interpMode=RCIM_Cubic, leaveTangent=7.8977, leaveTangentWeight=0.3558, tangentMode=RCTM_User, time=0, value=0}
         {arriveTangentWeight=0.2236, interpMode=RCIM_Cubic, leaveTangentWeight=0.1491, tangentMode=RCTM_User, time=0.3, value=1}
         {arriveTangentWeight=0.0667, interpMode=RCIM_Cubic, leaveTangentWeight=0.1667, tangentMode=RCTM_User, time=0.7, value=1}
         {arriveTangent=-11.5428, tangentMode=RCTM_User, time=1, value=0}
        ]
       }}}
     CurveIndex=$Emitter.NormalizedLoopAge
    )
    SpawnRate=&RandomRangeFloat(Minimum=10, Maximum=75, Evaluation Type=::NewEnumerator1)
    Use Spawn Probability=true
   )
  ]
  particleSpawnScript [
   /SetVariables(Particles.ScaleMod:NiagaraFloat=&FloatFromCurve(
     FloatCurve=DataInterface() {propertyValues {Curve {keys [
         {arriveTangent=-0.7341, interpMode=RCIM_Cubic, leaveTangent=-0.7341, tangentMode=RCTM_User, time=0, value=1}
         {arriveTangent=-0.6837, interpMode=RCIM_Cubic, leaveTangent=-0.6837, tangentMode=RCTM_User, time=1, value=0.35}
        ]}}}
     CurveIndex=&RandomRangeFloat(Minimum=0, Maximum=1)
    )) {isSetParameters=true}
   &InitializeParticle(
    Color=&RandomRangeLinearColor(Minimum=<rgba 1 1 1 0.75>, Maximum=<rgba 1 1 1 0.2>)
    Lifetime=&FloatFromCurve(
     FloatCurve=DataInterface() {propertyValues {Curve {keys [{arriveTangent=-0.9377, interpMode=RCIM_Cubic, leaveTangent=-0.9377, tangentMode=RCTM_User, time=0.05, value=0.55}, <at 0.9 0.2 RCIM_Cubic>]}}}
     CurveIndex=&RandomRangeFloat(Minimum=0, Maximum=1)
    )
    Sprite Size Mode=::NewEnumerator3
    Uniform Sprite Size=&Multiply_Float(A=&Multiply_Float(A=15, B=$Particles.ScaleMod), B=1)
    UsePositionOffset=false
    Write Material Random=true
   )
   &SocketLocation(
    Interpolate Socket Sampling=true
    SamplePosOffset=true
    Sampled Position Offset=<0 0 -10>
    Socket Index=1
    Socket Reader=$System.Socket Reader
    Socket Sampling Mode=::NewEnumerator3
    VelocitySpeedLimit=false
   )
   &ShapeLocation(MaskSpawnGroup=false, Override Seed=false, Sphere Radius=10, UseOwnerScale=true)
   &InheritVelocity(Inherited Velocity Amount Scale=<0.08 0.08 0.08>, Inherited Velocity Speed Limit=50, Speed Limit=true, UseSpeedThreshold=false)
   &AddVelocity(
    Rotation Coordinate Space=::World
    Velocity=$Output.SocketLocation.SampledVelocity
    Velocity Speed Scale=&RandomRangeFloat(Minimum=0.02, Maximum=0.08)
   )
   /SetVariables(Particles.float:NiagaraFloat=&RandomRangeFloat(Minimum=0, Maximum=360)) {isSetParameters=true}
   &CameraOffset(Camera Offset Amount=20)
   /SetVariables(Particles.DistanceScaleOpacity=1) {isSetParameters=true}
  ]
  particleUpdateScript [
   &ParticleState()
   &ScaleSpriteSize(
    Scale Factor=&Multiply_Vector2D_ByFloat(
     Float=&FloatFromCurve(FloatCurve=DataInterface() {propertyValues {Curve {
         defaultValue=1
         keys [
          {arriveTangent=0.0844, interpMode=RCIM_Cubic, leaveTangent=0.0844, leaveTangentWeight=0.3333, tangentMode=RCTM_User, time=0, value=0.53}
          {
           arriveTangent=0.3354
           arriveTangentWeight=0.1911
           interpMode=RCIM_Cubic
           leaveTangent=0.3354
           leaveTangentWeight=0.1422
           tangentMode=RCTM_User
           time=0.5734
           value=0.5966
          }
          {
           arriveTangent=0.7849
           arriveTangentWeight=0.0271
           interpMode=RCIM_Cubic
           leaveTangent=0.7849
           leaveTangentWeight=0.1151
           tangentMode=RCTM_User
           time=0.6546
           value=0.6043
          }
          {
           arriveTangent=0.6802
           arriveTangentWeight=0.0306
           interpMode=RCIM_Cubic
           leaveTangent=0.6802
           leaveTangentWeight=0.0846
           tangentMode=RCTM_User
           time=0.7463
           value=0.6851
          }
          {
           arriveTangent=0.4774
           arriveTangentWeight=0.0394
           interpMode=RCIM_Cubic
           leaveTangent=0.4774
           leaveTangentWeight=0.0452
           tangentMode=RCTM_User
           time=0.8644
           value=0.6515
          }
          {
           arriveTangent=12.3514
           arriveTangentWeight=0.0119
           interpMode=RCIM_Cubic
           leaveTangent=12.3514
           leaveTangentWeight=0.0333
           tangentMode=RCTM_User
           time=0.9
           value=1
          }
          {interpMode=RCIM_Cubic, tangentMode=RCTM_User, time=1, value=1}
         ]
        }}})
     Vector2D=&MakeVector2D(
      X=&FloatFromCurve(FloatCurve=DataInterface() {propertyValues {Curve {keys [
           {interpMode=RCIM_Cubic, leaveTangentWeight=0.2737, tangentMode=RCTM_User, time=0, value=1.35}
           {arriveTangent=-4.0805, interpMode=RCIM_Cubic, leaveTangent=-4.0805, tangentMode=RCTM_User, time=0.1813, value=1}
           {arriveTangent=-0.1492, interpMode=RCIM_Cubic, leaveTangent=-0.1492, time=0.2013, value=0.8992}
           <at 0.2613 0.9881 RCIM_Cubic>
          ]}}})
      Y=&FloatFromCurve(FloatCurve=DataInterface() {propertyValues {Curve {keys [
           {arriveTangent=-0.2248, interpMode=RCIM_Cubic, leaveTangent=-0.2248, tangentMode=RCTM_User, time=0, value=0.75}
           {arriveTangent=3.3994, interpMode=RCIM_Cubic, leaveTangent=3.3994, tangentMode=RCTM_User, time=0.1813, value=1}
           <at 0.2013 1.1617 RCIM_Cubic>
           <at 0.2613 1 RCIM_Cubic>
          ]}}})
     )
    )
    Scale Sprite Size Mode=::NewEnumerator2
   )
   /SetVariables(Particles.DistanceScaleOpacity:NiagaraFloat=/Niagara/DynamicInputs/Math/Min_Float(
     A:NiagaraFloat=$Particles.DistanceScaleOpacity
     B:NiagaraFloat=/Game/Effects/Niagara/DynamicInputs/Math/FN_Fit(
      Clamp:NiagaraBool=true
      Value:NiagaraFloat=&FloatFromCurve(
       FloatCurve=DataInterface() {propertyValues {Curve {keys [{interpMode=RCIM_Cubic, leaveTangentWeight=0.5912, tangentMode=RCTM_User, time=0.5, value=1}, <at 1.2 0 RCIM_Cubic>]}}}
       CurveIndex=&Divide_Float(
        A=&VectorLength(Vector=&SubtractVectorFromPosition(Position=$Particles.Position, Vector=Linked() {linkedVariable {name=Engine.Emitter.SimulationPosition, type=NiagaraPosition}}))
        B=100
       )
      )
      SourceMin:NiagaraFloat=0
      SourceMax:NiagaraFloat=1
      DestMin:NiagaraFloat=0
      DestMax:NiagaraFloat=1
     )
    )) {isSetParameters=true}
   &Drag(Drag=&RandomRangeFloat(Minimum=5, Maximum=13), Use Linear Drag=true, Use Rotational Drag=false)
   &GravityForce(Gravity=&MakeVector(
     X=0
     Y=0
     Z=&FloatFromCurve(
      FloatCurve=DataInterface() {propertyValues {Curve {keys [<at 0.8 150 RCIM_Constant>,<at 0.9 -200 RCIM_Constant>]}}}
      Scale Curve=&RandomRangeFloat(Minimum=0.9, Maximum=1.2)
     )
    ))
   &CurlNoiseForce(
    Bias Noise=false
    Noise Strength=&FloatFromCurve(
     FloatCurve=DataInterface() {propertyValues {Curve {
        defaultValue=1
        keys [
         {arriveTangent=6.7916, interpMode=RCIM_Cubic, leaveTangent=6.7916, tangentMode=RCTM_User, time=0.1, value=0}
         {arriveTangentWeight=0.2399, interpMode=RCIM_Cubic, leaveTangentWeight=0.2938, tangentMode=RCTM_User, time=0.4, value=1}
         {arriveTangentWeight=0.1795, interpMode=RCIM_Cubic, leaveTangentWeight=0.1795, tangentMode=RCTM_User, time=0.7, value=1}
         {arriveTangent=-9.5495, interpMode=RCIM_Cubic, leaveTangent=-9.5495, tangentMode=RCTM_User, time=0.9, value=0}
        ]
       }}}
     Scale Curve=90
    )
    Pan Noise=false
   )
   &SolveForcesAndVelocity(Clamp Velocity=false, Limit Acceleration=false)
   &ScaleColor(
    Scale Alpha=&Multiply_Float(
     A=&Multiply_Float(
      A=&FloatFromCurve(
       FloatCurve=DataInterface() {propertyValues {Curve {keys [
           {arriveTangent=0.8727, interpMode=RCIM_Cubic, leaveTangent=0.8727, tangentMode=RCTM_User, time=0, value=0}
           {arriveTangent=1.859, interpMode=RCIM_Cubic, leaveTangent=1.859, tangentMode=RCTM_User, time=0.2882, value=0.2618}
           {
            arriveTangent=13.7023
            arriveTangentWeight=0.1345
            interpMode=RCIM_Cubic
            leaveTangentWeight=0.1345
            tangentMode=RCTM_Break
            time=0.4
            value=1
           }
           {arriveTangentWeight=0.1274, interpMode=RCIM_Cubic, leaveTangentWeight=0.2613, tangentMode=RCTM_User, time=0.8, value=1}
           {arriveTangent=-18.5888, tangentMode=RCTM_User, time=0.9, value=0}
          ]}}}
       Scale Curve=/Niagara/DynamicInputs/Camera/CameraFade(
        Camera Query:NiagaraDataInterfaceCamera=DataInterface()
        Fade Offset:NiagaraFloat=100
        Fade Distance:NiagaraFloat=100
        Use Owner Position:NiagaraBool=true
       )
      )
      B=&FloatFromCurve(
       FloatCurve=DataInterface() {propertyValues {Curve {keys [<at 0 0.75>,<at 0.3925 1>,<at 0.5964 1>,<at 1 0.75>]}}}
       CurveIndex=Linked() {linkedVariable {name=NPC.FortniteNPC.TimeOfDay_0-1, type=NiagaraFloat}}
       Scale Curve=&FN_FrontEndFloatModifier(Gameplay=1, Front End=0.7)
      )
     )
     B=$Particles.DistanceScaleOpacity
    )
    Scale RGB=&Multiply_VectorByFloat(Vector=<1 1 1>, Float=1)
    ScaleA=true
    ScaleRGB=true
   )
   /SetVariables(
    Particles.Splash Color:LinearColor=&MakeLinearColorFromVectorAndFloat(
     "Vector (RGB)"=<1 1 1>
     "Float (Alpha)"=&FloatFromCurve(
      FloatCurve=DataInterface() {propertyValues {Curve {keys [
          <at 0.88 0 RCIM_Cubic>
          {arriveTangent=-38.7858, interpMode=RCIM_Cubic, leaveTangent=-38.7858, tangentMode=RCTM_User, time=0.9, value=2}
          <at 1 0 RCIM_Cubic>
         ]}}}
      Scale Curve=&MakeCustomFloatFromBool(Bool=$User.bIsFrontEndPreview, True Float=0.7, False Float=1)
     )
    )
    Particles.Splash Dyn 0:Vector4f=&MakeVector4(
     X=&FloatFromCurve(FloatCurve=DataInterface() {propertyValues {Curve {
         defaultValue=1
         keys [
          {interpMode=RCIM_Cubic, tangentMode=RCTM_User, time=0.9, value=0}
          {arriveTangent=14.9612, interpMode=RCIM_Cubic, leaveTangent=14.9612, tangentMode=RCTM_User, time=1, value=0.5786}
         ]
        }}})
     Y=0
     Z=0
     W=0
    )
    Particles.Splash Size:Vector2f=&Multiply_Vector2D(
     A=$Particles.Initial.SpriteSize
     B=&Vector2DFromFloat(Value=&FloatFromCurve(
       FloatCurve=DataInterface() {propertyValues {Curve {keys [
           {arriveTangent=19.4142, interpMode=RCIM_Constant, time=0.85, value=0}
           {arriveTangentWeight=1.0543, interpMode=RCIM_Cubic, leaveTangentWeight=14.76, tangentMode=RCTM_User, time=0.9, value=0.5}
           {arriveTangent=-1.8786, interpMode=RCIM_Cubic, leaveTangent=-1.8786, tangentMode=RCTM_User, time=1, value=1}
          ]}}}
       Scale Curve=1.75
      ))
    )
   ) {isSetParameters=true}
   &ScaleMeshSize(Scale Factor=&Multiply_VectorByFloat(
     Vector=<1 1 1>
     Float=&FloatFromCurve(FloatCurve=DataInterface() {propertyValues {Curve {keys [
          {arriveTangent=0.0844, interpMode=RCIM_Cubic, leaveTangent=0.0844, tangentMode=RCTM_User, time=0, value=0.5}
          {arriveTangent=0.3354, interpMode=RCIM_Cubic, leaveTangent=0.3354, tangentMode=RCTM_User, time=0.5734, value=0.5966}
          {arriveTangent=0.7849, interpMode=RCIM_Cubic, leaveTangent=0.7849, tangentMode=RCTM_User, time=0.6546, value=0.6043}
          {arriveTangent=0.6802, interpMode=RCIM_Cubic, leaveTangent=0.6802, tangentMode=RCTM_User, time=0.7463, value=0.6851}
          {arriveTangent=0.4774, interpMode=RCIM_Cubic, leaveTangent=0.4774, tangentMode=RCTM_User, time=0.8644, value=0.6515}
          {arriveTangent=12.3514, interpMode=RCIM_Cubic, leaveTangent=12.3514, tangentMode=RCTM_User, time=0.9, value=1}
         ]}}})
    ))
   &DynamicMaterialParameters(
    Index 0 Param 1=&FloatFromCurve(FloatCurve=DataInterface() {propertyValues {Curve {
        defaultValue=1
        keys [
         {arriveTangent=14.7642, interpMode=RCIM_Cubic, leaveTangent=14.7642, tangentMode=RCTM_User, time=0, value=0}
         {arriveTangent=-2.5249, interpMode=RCIM_Cubic, leaveTangent=-2.5249, tangentMode=RCTM_User, time=0.1, value=0.7}
         {arriveTangent=-0.0405, interpMode=RCIM_Cubic, leaveTangent=-0.0405, tangentMode=RCTM_User, time=0.5, value=0}
        ]
       }}})
    Index 0 Param 2=&FloatFromCurve(FloatCurve=DataInterface() {propertyValues {Curve {
        defaultValue=1
        keys [
         {interpMode=RCIM_Cubic, leaveTangentWeight=0.9127, tangentMode=RCTM_User, time=0.2, value=0}
         {arriveTangentWeight=0.2716, interpMode=RCIM_Cubic, leaveTangentWeight=0.4526, tangentMode=RCTM_User, time=0.5, value=0.2}
         {arriveTangent=-0.0405, interpMode=RCIM_Cubic, leaveTangent=-0.0405, tangentMode=RCTM_User, time=1, value=0}
        ]
       }}})
    Index 0 Param 3=&FloatFromCurve(FloatCurve=DataInterface() {propertyValues {Curve {keys [<at 0.25 1>,<at 0.3 0>]}}})
    Module.Index0Float0Write=true
    Module.Index0Float1Write=true
    Module.Index0Float2Write=true
    Module.Index0Float3Write=true
   )
   &KillParticles(Kill Particles=&SetBoolByFloatComparison(A=$Particles.DistanceScaleOpacity, B=0.001, Comparison Type=::NewEnumerator7))
  ]
  rendererClasses [&NiagaraSpriteRendererProperties]
  renderers [
   {rendererClass=&NiagaraSpriteRendererProperties}
   {rendererClass=&NiagaraSpriteRendererProperties, rendererIndex=1}
  ]
 }
 Hard_Lines {
  emitterUpdateScript [
   &EmitterState(Life Cycle Mode=::NewEnumerator1, Loop Behavior=::NewEnumerator1, Loop Duration=0.22, UseLoopDelay=true)
   &FN_SpawnAnimTrail_v2(AnimTrailWidthMode=::FromSecond)
  ]
  particleSpawnScript [
   /Game/Effects/Niagara/Modules/AnimTrail/FN_InitializeAnimTrail/FN_InitializeAnimTrailBoneIndecies(First Bone or Socket Index=1)
   /Game/Effects/Niagara/Modules/AnimTrail/FN_InitializeAnimTrail/FN_InitializeAnimTrail(AnimTrailWidth=0.55, Color=<rgba 0 0 0 0.4>, Lifetime=0.15)
   &SystemLocation(MaskSpawnGroup=false)
  ]
  particleUpdateScript [
   &ScaleRibbonWidth(Ribbon Width Scale=&FloatFromCurve(FloatCurve=DataInterface() {propertyValues {Curve {keys [
         {arriveTangent=-1.2579, interpMode=RCIM_Cubic, time=0.7, value=1}
         {arriveTangent=-0.3747, interpMode=RCIM_Cubic, leaveTangent=-0.3747, tangentMode=RCTM_User, time=1.0133, value=0.8}
        ]}}}))
   &ScaleColor(
    Scale Alpha=&Multiply_Float(
     A=&FloatFromCurve(
      FloatCurve=DataInterface() {propertyValues {Curve {keys [
          {interpMode=RCIM_Cubic, leaveTangent=-4.9655, tangentMode=RCTM_Break, time=0, value=1}
          <at 0.2 0.0233>
          {arriveTangent=-0.2139, interpMode=RCIM_Cubic, tangentMode=RCTM_Break, time=1, value=0}
         ]}}}
      Scale Curve=&FN_FrontEndFloatModifier(Gameplay=1, Front End=0.6)
     )
     B=&FloatFromCurve(
      FloatCurve=DataInterface() {propertyValues {Curve {keys [<at 0 0.9 RCIM_Cubic>,<at 0.4982 1 RCIM_Cubic>,<at 1 0.9 RCIM_Cubic>]}}}
      CurveIndex=Linked() {linkedVariable {name=NPC.FortniteNPC.TimeOfDay_0-1, type=NiagaraFloat}}
     )
    )
    ScaleA=true
    ScaleRGB=true
   ) {module=ScaleColor002}
   &FN_UpdateAnimTrail_v2()
  ]
  rendererClasses [&NiagaraRibbonRendererProperties]
  renderers [{rendererClass=&NiagaraRibbonRendererProperties}]
 }
 Trail_Instrument_Color {
  emitterUpdateScript [
   &EmitterState(Life Cycle Mode=::NewEnumerator1, Loop Behavior=::NewEnumerator1, Loop Duration=0.22, UseLoopDelay=true)
   &FN_SpawnAnimTrail_v2()
  ]
  particleSpawnScript [
   &FN_InitializeAnimTrailBoneIndecies(First Bone or Socket Index=1)
   &FN_InitializeAnimTrail(AnimTrailWidth=1, Color=<rgba 1 1 1 0.7>, Lifetime=0.16)
   &SystemLocation(MaskSpawnGroup=false)
  ]
  particleUpdateScript [
   &ScaleRibbonWidth(Ribbon Width Scale=&FloatFromCurve(FloatCurve=DataInterface() {propertyValues {Curve {keys [
         {arriveTangent=-1.2579, interpMode=RCIM_Cubic, time=0, value=1}
         {arriveTangent=-0.9664, interpMode=RCIM_Cubic, leaveTangent=-0.9664, tangentMode=RCTM_User, time=1, value=0}
        ]}}}))
   &ScaleColor(
    Scale Alpha=&Multiply_Float(
     A=&FloatFromCurve(
      FloatCurve=DataInterface() {propertyValues {Curve {keys [
          {interpMode=RCIM_Cubic, leaveTangent=-2.0543, tangentMode=RCTM_Break, time=0, value=1}
          {arriveTangent=-1, interpMode=RCIM_Cubic, leaveTangent=-1, time=0.5, value=0.2}
          {arriveTangent=-0.2139, interpMode=RCIM_Cubic, tangentMode=RCTM_Break, time=1, value=0}
         ]}}}
      Scale Curve=&FloatFromCurve(
       FloatCurve=DataInterface() {propertyValues {Curve {keys [<at 0 1>,<at 0.25 1>,<at 0.5 1>,<at 0.75 1>,<at 1 1>]}}}
       CurveIndex=Linked() {linkedVariable {name=NPC.FortniteNPC.TimeOfDay_0-1, type=NiagaraFloat}}
      )
     )
     B=&FN_FrontEndFloatModifier(Gameplay=1, Front End=0.6)
    )
    Scale RGB=&MakeVectorFromLinearColorRGB(Color=$System.TOD Color)
    ScaleA=true
    ScaleRGB=true
   ) {module=ScaleColor002}
   &FN_UpdateAnimTrail_v2()
  ]
  rendererClasses [&NiagaraRibbonRendererProperties]
  renderers [{rendererClass=&NiagaraRibbonRendererProperties}]
 }
}
summary {
 systemName=NS_Pickaxe_DrumStick_CountyFrog_Trail
 emitters [
  {emitterName=Trail_Instrument_Color, rendererClasses [&NiagaraRibbonRendererProperties]}
  {emitterName=Hard_Lines, rendererClasses [&NiagaraRibbonRendererProperties]}
  {emitterName=Bubble_Sprites001, rendererClasses [&NiagaraSpriteRendererProperties]}
 ]
 userVariables [
  User.bIsFrontEnd {type=NiagaraBool}
  User.bIsFrontEndPreview {type=NiagaraBool}
  User.IsFrontEnd {type=NiagaraBool}
  User.RibbonIDOffset {type=NiagaraInt32}
  User.SkeletalMesh {
   type=NiagaraDataInterfaceSkeletalMesh
   defaultValue {
    struct=NiagaraExt_VariableValue_DataInterface
    value {dataInterface=&self:NiagaraDataInterfaceSkeletalMesh_0}
   }
  }
 ]
}
